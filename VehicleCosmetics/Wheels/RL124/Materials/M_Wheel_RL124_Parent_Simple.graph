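# Simple parent material for the RL124 vehicle wheel, set up for skeletal meshes with Default Lit shading.
# The wheel's material function feeds vehicle damage and de-rez transition functions, which produce the surface attributes.
# World Position Offset uses the TirePop parameter to pull vertices toward the object position, masked by green vertex color.
# The vertex-normal branch of that blend is multiplied by zero, so unmasked vertices stay in place.
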Material /VehicleCosmetics/Wheels/RL124/Materials/M_Wheel_RL124_Parent_Simple
graph {
 properties {
  bAutomaticallySetUsageInEditor=false
  bUseMaterialAttributes=true
  bUsedWithMeshDeformer=true
  bUsedWithSkeletalMesh=true
  pixelDepthOffsetMode=PDOM_Legacy
  usedShadingModels=Default Lit
 }
 nodes [
  n1 = MaterialFunctionCall() {materialFunction=&/MF_Wheel_RL124_Parent_Simple}
  n2 = VertexNormalWS()
  n3 = ScalarParameter() {parameterName=TirePop}
  n4 = Subtract(A=ObjectPositionWS(), B=WorldPosition().XYZ)
  n5 = VertexColor()
  SetMaterialAttributes!(
   MaterialAttributes=MaterialFunctionCall(In=MaterialFunctionCall(Material=@n1.MaterialAttributes) {materialFunction=&MF_BaseVehicle_Damage}.Result) {materialFunction=&MF_VehicleDeRez_Transition}.Result
   World Position Offset=LinearInterpolate(A=Multiply(A=@n2) {constB=0}, B=Multiply(A=@n3, B=@n4), Alpha=Saturate(None=@n5.G))
  ) {attributeSetTypes [#1]}
 ]
}
